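# Anim Blueprint for the BuffBeak backpack cosmetic that adds secondary motion to its keychain and zipper pulls.
# A bRigidBodyEnabled bool switches between a RigidBody simulation using the pack's physics asset and a three-bone keychain AnimDynamics chain.
# AnimDynamics then drives both zipper pulls, and an additive bone tweak offsets the left one.
# The event graph computes a deprecated gravity override and a left-zipper alpha, neither visibly read by the AnimGraph.
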AnimBlueprint /BRCosmetics/Accessories/FORT_Backpacks/Backpack_F_MED_BuffBeak/Meshes/F_MED_BuffBeak_Pack_AnimBP
target_skeleton=&/F_MED_BuffBeak_Pack_Skeleton
graphs [
 AnimGraph {nodes [
   n1 = RigidBody!(ComponentPose=Local To Component(LocalPose=Input Pose()), Alpha, OverrideWorldGravity:Vector="0,0,0", ExternalForce) -> (Pose) {
    bOverrideWorldGravity=true
    overridePhysicsAsset=&/F_MED_BuffBeak_Pack_Physics
    alphaBoolBlend {blendOption=Linear}
    componentAppliedLinearAccClamp=<600 600 600>
    componentLinearAccScale=<0.4 0.4 0.4>
    componentLinearVelScale=<0.4 0.4 0.4>
   }
   n2 = Component To Local(ComponentPose=@n1)
   n3 = Anim Dynamics - Bone: dyn_keychain_01!(ComponentPose=Local To Component(LocalPose=Input Pose() {name=InPose_0}), Alpha=1, GravityOverride:Vector="0,0,-980", ExternalForce="0,0,0") -> (Pose) {
    angularDampingOverride=0.98
    bOverrideAngularDamping=true
    bUseGravityOverride=true
    alphaBoolBlend {blendOption=Linear}
    boundBone {boneName=dyn_keychain_01}
    componentAppliedLinearAccClamp=<1000 1000 1000>
    componentLinearAccScale=<0.5 0.5 0.5>
    componentLinearVelScale=<0.5 0.5 0.5>
    gravityOverride {z=-980}
    physicsBodyDefinitions [{
      boundBone {boneName=dyn_keychain_01}
      boxExtents=<4 12 12>
      constraintSetup {angularLimitsMax=<15 30 30>, angularLimitsMin=<-15 -30 -30>, angularTarget=<0 0 0>}
      localJointOffset=<1 0 0>
     }]
   }
   n4 = Anim Dynamics - Bone: dyn_keychain_02!(ComponentPose=@n3, Alpha=1, GravityOverride:Vector="0,0,-980", ExternalForce="0,0,0") -> (Pose) {
    angularDampingOverride=0.98
    bOverrideAngularDamping=true
    bUseGravityOverride=true
    alphaBoolBlend {blendOption=Linear}
    boundBone {boneName=dyn_keychain_02}
    componentAppliedLinearAccClamp=<800 800 800>
    componentLinearAccScale=<0.5 0.5 0.5>
    componentLinearVelScale=<0.5 0.5 0.5>
    gravityOverride {z=-980}
    physicsBodyDefinitions [{
      boundBone {boneName=dyn_keychain_02}
      boxExtents=<4 10 10>
      constraintSetup {angularLimitsMax=<15 25 25>, angularLimitsMin=<-15 -25 -25>, angularTarget=<0 0 0>}
      localJointOffset=<1 0 0>
     }]
   }
   n5 = Anim Dynamics - Bone: dyn_keychain_03!(ComponentPose=@n4, Alpha=1, GravityOverride:Vector="0,0,-980", ExternalForce="0,0,0") -> (Pose) {
    angularDampingOverride=0.99
    bOverrideAngularDamping=true
    bUseGravityOverride=true
    bUseSphericalLimits=true
    alphaBoolBlend {blendOption=Linear}
    boundBone {boneName=dyn_keychain_03}
    componentAppliedLinearAccClamp=<650 650 650>
    componentLinearAccScale=<0.5 0.5 0.5>
    componentLinearVelScale=<0.5 0.5 0.5>
    gravityOverride {z=-980}
    physicsBodyDefinitions [{
      boundBone {boneName=dyn_keychain_03}
      boxExtents=<5 10 10>
      constraintSetup {angularLimitsMax=<15 25 25>, angularLimitsMin=<-15 -25 -25>, angularTarget=<0 0 0>}
      localJointOffset=<1 0 0>
     }]
    sphericalLimits [{limitRadius=5, drivingBone {boneName=dyn_keychain_MSTR}, sphereLocalOffset=<-2.0592 0 2.3>}]
   }
   n6 = Component To Local(ComponentPose=@n5)
   n7 = Get bRigidBodyEnabled()
   n8 = Anim Dynamics - Bone: dyn_zipper_01_r!(
    ComponentPose=Local To Component(LocalPose=Blend Poses by bool(BlendPose_0=@n2, BlendPose_1=@n6, bActiveValue=@n7) {blendType=Linear})
    Alpha=1
    GravityOverride:Vector="0,0,-980"
    ExternalForce="0,0,0"
   ) -> (Pose) {
    angularDampingOverride=0.9999
    bGravityOverrideInSimSpace=true
    bOverrideAngularDamping=true
    bUseGravityOverride=true
    alphaBoolBlend {blendOption=Linear}
    boundBone {boneName=dyn_zipper_01_r}
    componentAppliedLinearAccClamp=<1100 1100 1100>
    componentLinearAccScale=<1 1 1>
    componentLinearVelScale=<1 1 1>
    gravityOverride {z=-980}
    physicsBodyDefinitions [{
      sphereCollisionRadius=0
      boundBone {boneName=dyn_zipper_01_r}
      boxExtents=<10 5 5>
      constraintSetup {angularLimitsMax=<0 30 0>, angularLimitsMin=<0 -15 -120>, angularTarget=<0 0 0>}
      localJointOffset=<1 0 0>
     }]
   }
   n9 = Anim Dynamics - Bone: dyn_zipper_01_l!(ComponentPose=@n8, Alpha=1, GravityOverride:Vector="0,0,-980", ExternalForce="0,0,0") -> (Pose) {
    angularDampingOverride=0.9999
    bGravityOverrideInSimSpace=true
    bOverrideAngularDamping=true
    bUseGravityOverride=true
    alphaBoolBlend {blendOption=Linear}
    boundBone {boneName=dyn_zipper_01_l}
    componentAppliedLinearAccClamp=<700 700 700>
    componentLinearAccScale=<1 1 1>
    componentLinearVelScale=<1 1 1>
    gravityOverride {z=-980}
    physicsBodyDefinitions [{
      sphereCollisionRadius=0
      boundBone {boneName=dyn_zipper_01_l}
      boxExtents=<12 5 5>
      constraintSetup {angularLimitsMax=<0 100 0>, angularLimitsMin=<0 -5 -30>, angularTarget=<0 0 0>}
      localJointOffset=<1 0 0>
     }]
   }
   n10 = "Transform (Modify) Bone - Bone: dyn_zipper_01_l"(ComponentPose=@n9, Alpha=1, Translation="0,0,-0.5", Rotation="0,0,-10", Scale="1,1,1") {
    rotationMode=BMM_Additive
    translationMode=BMM_Additive
    alphaBoolBlend {blendOption=Linear}
    boneToModify {boneName=dyn_zipper_01_l}
    rotation {roll=-10}
    translation {z=-0.5}
   }
   Output Pose(Result=Component To Local(ComponentPose=@n10)) {layerGroup=None}
  ]}
 EventGraph {nodes [
   Event BlueprintUpdateAnimation()
   > n1 = Gravity Override State Function DEPRECATED(CharacterPartAnimInstance=Self-Reference(), GameTimeInSeconds=GetGameTimeInSeconds(), GravityOverrideParams=Get Gravity Override Params())
   > Set Graviti Override(Graviti Override:Vector=@n1) -> (Output_Get:Vector)
   > Set Alpha Zipper 01 L(Alpha Zipper 01 L:double=GetAlphaFromReferenceTransformDelta(
     SkeletalMeshComponent=Get Skel Mesh()
     BoneName=dyn_zipper_01_l
     BaseName=dyn_zipper_MSTR_l
     DeltaComponentType=EulerY
     InputRange=(X=-10,Y=-70)
     TwistAxis="1,0,0"
    )) -> (Output_Get:double)
   Event BlueprintInitializeAnimation()
   > Set Skel Mesh(Skel Mesh=GetPartSkeletalMeshComponent())
   TryGetPawnOwner()
  ]}
]
variables [
 Alpha Zipper 01 L {type=number}
 Graviti Override {type=Vector}
 Gravity Override Params {type=GravityOverrideParamsStruct}
 Skel Mesh {type=SkeletalMeshComponent}
]
